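# Physics asset for the M_MED_GoldCat backpack cosmetic, setting up its mouse, ring and zipper parts for secondary motion.
# All bodies are capsules, with a three-capsule pack body that has no constraints.
# The ring chains to the mouse, which has tight swing limits against its MSTR body, while the zipper swings widely from its own MSTR body.
# All three constraints set zero twist limits, and the ring also has zero second swing.
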PhysicsAsset /BRCosmetics/Accessories/FORT_Backpacks/Backpack_M_MED_GoldCat/Meshes/M_MED_GoldCat_Pack_Physics
bodies {
 dyn_mouse [
  {shapeType=Capsule, length=4.1, radius=3.2, shapeName=dyn_mouse_capsule, center=<-2 0.1927 0>, rotation=<rot 90 0 0>}
  {shapeType=Capsule, length=0.4, radius=4.2, shapeName=dyn_mouse_capsule, center=<-6.5092 0.4254 0>, rotation=<rot 90 0 0>}
 ]
 dyn_mouse_MSTR [{shapeType=Capsule, length=1, radius=1, shapeName=dyn_mouse_MSTR_capsule}]
 dyn_ring_01 [{shapeType=Capsule, length=0, radius=2.4, shapeName=dyn_ring_01_capsule, center=<2 0 0>, rotation=<rot 90 0 0>}]
 dyn_zipper_01 [{shapeType=Capsule, length=2, radius=2, shapeName=dyn_zipper_01_capsule, center=<2.311 0 0>, rotation=<rot 90 0 0>}]
 dyn_zipper_MSTR [{shapeType=Capsule, length=1, radius=1, shapeName=dyn_zipper_MSTR_capsule}]
 pack [
  {shapeType=Capsule, length=5.5, radius=5.9, shapeName=pack_capsule, center=<-12.3006 1.6677 -6.0171>}
  {shapeType=Capsule, length=9.2, radius=5.1, shapeName=pack_capsule, center=<-6.5888 3.7 6.5>, rotation=<rot 90 0 0>}
  {shapeType=Capsule, length=3.1, radius=1.7, shapeName=pack_capsule, center=<-11.3257 -6.3445 0.6777>, rotation=<rot 64.4271 -0.0452 -0.0502>}
 ]
}
constraints [
 {bone1Name=dyn_zipper_01, bone2Name=dyn_zipper_MSTR, swing1LimitDegrees=110, swing2LimitDegrees=110, twistLimitDegrees=0}
 {bone1Name=dyn_mouse, bone2Name=dyn_mouse_MSTR, swing1LimitDegrees=15, swing2LimitDegrees=5, twistLimitDegrees=0}
 {bone1Name=dyn_ring_01, bone2Name=dyn_mouse, swing1LimitDegrees=120, swing2LimitDegrees=0, twistLimitDegrees=0}
]
